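; Offscreen capture actor for the weapon mod station UI, rendering a preview weapon and its mods into main and mask render targets.
; It spawns a preview copy on a separate lighting channel, fits camera distance to mesh bounds, and previews, confirms or reverts mods.
; Each capture briefly shows the weapon; full captures also swap in mask materials marking the selected slot, then restore originals.
; Render target size and lights vary by platform.
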(event Rendering|EventOnRefreshCapture
  (CaptureWeaponImage|CaptureWeapon false))

(event EventBeginPlay)

(event Custom|Update%20MainPass
  (CaptureWeaponImage|CaptureWeapon true))

(event Custom|Event%20DoubleTake
  (if (Variables|Default|GetDoDoubleTake)
    (Utilities|FlowControl|Delay 0.005)
    (CaptureWeaponImage|CaptureWeapon false)
    (Variables|Default|SetDoDoubleTake false)))

(fn ApplyWeaponWrap
  (params (Equipped%20Weapon :type Fort%20Weapon%20Object%20Reference))
  (Utilities|IsValid Equipped%20Weapon
    (:Is%20Valid
      (for (_array_element _array_index) (Rendering|Material|GetMaterials (Variables|/Script/FortniteGame.FortWeapon|GetWeaponMesh Equipped%20Weapon))
        (Rendering|Material|SetMaterial (Variables|/Script/FortniteGame.FortWeapon|GetWeaponMesh (Variables|Default|GetPreviewWeapon)) _array_index _array_element)))))

(fn AssignMaskMaterials
  (bind _local_mod_comp (LocalVariables|GetLocal%20Mod%20Comp))
  (bind _translucentmaterialmask (Variables|Default|GetTranslucentMaterialMask))
  (bind _attachedmodsmeshcomponents (Variables|Default|GetAttachedModsMeshComponents))
  (bind _previewweapon (Variables|Default|GetPreviewWeapon))
  (bind _weaponmesh (Variables|/Script/FortniteGame.FortWeapon|GetWeaponMesh _previewweapon))
  (Utilities|IsValid _previewweapon
    (:Is%20Valid
      (bind _output_get (Variables|Default|SetPreviewWeaponOriginalMaterials (Rendering|Material|GetMaterials _weaponmesh)))
      (for (_array_element _array_index) _output_get
        (bind _returnvalue (Rendering|Material|CreateDynamicMaterialInstance _weaponmesh _array_index "/WeaponModStation/WeaponUICapture/Materials/MI_Mask_Unselected.MI_Mask_Unselected" "None" :!class /Script/Engine.PrimitiveComponent))
        (bind _asmaterial_instance (Utilities|Casting|CastToMaterialInstance _array_element))
        (Utilities|ReadandApplyParameters _asmaterial_instance _returnvalue))
      (bind _keys (Utilities|Map|Keys (Variables|Default|GetAttachedModsMeshComponents)))
      (for _array_element_2 _keys
        (if (.ReturnValue (Utilities|Map|Find _attachedmodsmeshcomponents _array_element_2))
          (LocalVariables|SetLocal%20Mod%20Comp (Utilities|Map|Find _attachedmodsmeshcomponents _array_element_2))
          (for _index (range (Rendering|Material|GetNumMaterials (LocalVariables|GetLocal%20Mod%20Comp)))
            (LocalVariables|SetLocal%20Element%20Index _index)
            (switch Utilities|FlowControl|Switch|SwitchonEBlendMode (Rendering|Material|GetBlendMode (Rendering|Material|GetMaterial _local_mod_comp _index))
              (:BLEND_Opaque
                _)
              (:BLEND_Masked
                _)
              (:BLEND_Translucent
                (if (.ReturnValue (Utilities|Map|Find _translucentmaterialmask (Utilities|ResolveSoftReference (Rendering|Material|GetMaterial _local_mod_comp _index)))) :!label _label
                  (Rendering|Material|SetMaterial (LocalVariables|GetLocal%20Mod%20Comp) (LocalVariables|GetLocal%20Element%20Index) (Utilities|Map|Find _translucentmaterialmask (Utilities|ResolveSoftReference (Rendering|Material|GetMaterial _local_mod_comp _index))))
                  (else
                    (Rendering|Material|SetMaterial (LocalVariables|GetLocal%20Mod%20Comp) (LocalVariables|GetLocal%20Element%20Index) "/Game/Global/Materials/M_Overlay_VertexCrunch.M_Overlay_VertexCrunch"))))
              (:BLEND_Additive
                (goto _label))
              (:BLEND_Modulate
                (goto _label)))
            (Rendering|Material|SetMaterial (LocalVariables|GetLocal%20Mod%20Comp) (LocalVariables|GetLocal%20Element%20Index) (select (Utilities|String|Contains (Utilities|String|ToString(Name) _array_element_2) (Utilities|String|ToString(Name) (Variables|Default|GetSelectedModSlot))) "/WeaponModStation/WeaponUICapture/Materials/MI_Mask_Selected.MI_Mask_Selected" "/WeaponModStation/WeaponUICapture/Materials/MI_Mask_Unselected.MI_Mask_Unselected"))))))))

(fn CaptureWeapon
  (params (MainPass%20Only :type Boolean) (Selected%20Mod :type Enum_WeaponUICaptureSelectedMod%20Enum :default "NewEnumerator0"))
  (bind _selectedmodenum (Variables|Default|GetSelectedModEnum))
  (bind (_value _returnvalue) (Utilities|Map|Find _selectedmodenum Selected%20Mod))
  (if _returnvalue
    (Variables|Default|SetSelectedModSlot _value)
    (if MainPass%20Only
      ;[1 1 1] Main pass capture
      (Helpers|SetCaptureVisibility true)
      (Variables|/Script/Engine.SceneCaptureComponent2D|SetTextureTarget (Variables|Default|GetMainPass) (Utilities|SceneCaptureSelector))
      (Helpers|SceneCapture false)
      (Helpers|SetCaptureVisibility false)
      (else
        ;[1 1 1] Main pass capture
        (Helpers|SetCaptureVisibility true)
        (Helpers|SceneCapture false)
        ;[1 1 1] Masks Pass capture
        (Helpers|AssignMaskMaterials)
        (Helpers|SceneCapture true)
        (if (TouchPlatform|IsMobileGame)
          (Rendering|RefreshCapturewhenPSOPrecachingDone (Variables|/Script/FortniteGame.FortWeapon|GetWeaponMesh (Variables|Default|GetPreviewWeapon))))
        (Helpers|ResetMaterials)
        (Helpers|SetCaptureVisibility false)))
    (return (Variables|Default|GetMainPass) (Variables|Default|GetMasksPass))
    (else
      (Development|PrintString "Selected Mod isn't valid, please add it to BP_WeaponUICapture -> SelectedModEnum" true true "(R=1.000000,G=0.114804,B=0.000000,A=1.000000)" 1e+01))))

(fn ConfirmMod
  (params (Recapture%20Equipped%20Weapon :type Boolean :default true))
  (bind _previewweapon (Variables|Default|GetPreviewWeapon))
  (if Recapture%20Equipped%20Weapon
    (CaptureWeaponImage|ConstructWeapon 0)
    (else
      (Utilities|IsValid _previewweapon
        (:Is%20Valid
          (for _array_element (Class|FortWeapon|GetWeaponModSlots _previewweapon)
            (Utilities|Map|Add (Variables|Default|GetAttachedModsItemDefinition) (GameplayTags|GetTagName (Variables|/Script/FortniteGame.FortWeaponModItemDefinition|GetModSlot (Utilities|Struct|BreakFortWeaponModSlot _array_element))) (Utilities|Struct|BreakFortWeaponModSlot _array_element))
            (bind _returnvalue (Class|FortWeaponModFunctionLibrary|TryAddPreviewWeaponMod (Utilities|Struct|BreakFortWeaponModSlot _array_element) (Variables|Default|GetPreviewWeapon)))
            (Utilities|Map|Add (Variables|Default|GetAttachedModsMeshComponents) (GameplayTags|GetTagName (Variables|/Script/FortniteGame.FortWeaponModItemDefinition|GetModSlot (Utilities|Struct|BreakFortWeaponModSlot _array_element))) _returnvalue)
            (Rendering|Components|SetLightingChannels _returnvalue false false true)))))))

(fn ConstructWeapon
  (params (Equipped%20Weapon :type Fort%20Weapon%20Object%20Reference))
  (bind _previewweapon (Variables|Default|GetPreviewWeapon))
  (Utilities|Map|Clear (Variables|Default|GetAttachedModsItemDefinition))
  (Utilities|Map|Clear (Variables|Default|GetAttachedModsMeshComponents))
  (LocalVariables|SetLocalEquippedWeapon Equipped%20Weapon)
  (Utilities|IsValid _previewweapon
    (:Is%20Valid
      (Actor|DestroyActor _previewweapon))
    (:Is%20Not%20Valid
      _))
  ;[1 1 1] Get Mods Meshes
  (bind _returnvalue (Class|FortWeaponModFunctionLibrary|SpawnWeaponModPreviewActor (Data|GetWeaponData Equipped%20Weapon) (Transformation|GetActorTransform)))
  (Variables|Default|SetPreviewWeapon _returnvalue)
  (Class|FortWeaponModFunctionLibrary|RemoveAllPreviewWeaponMods (Variables|Default|GetPreviewWeapon))
  (for _array_element (Class|FortWeapon|GetWeaponModSlots Equipped%20Weapon)
    (bind _returnvalue_2 (Class|FortWeaponModFunctionLibrary|TryAddPreviewWeaponMod (Utilities|Struct|BreakFortWeaponModSlot _array_element) (Variables|Default|GetPreviewWeapon)))
    (Rendering|Components|SetLightingChannels _returnvalue_2 false false true)
    (Utilities|Map|Add (Variables|Default|GetAttachedModsItemDefinition) (GameplayTags|GetTagName (Variables|/Script/FortniteGame.FortWeaponModItemDefinition|GetModSlot (Utilities|Struct|BreakFortWeaponModSlot _array_element))) (Utilities|Struct|BreakFortWeaponModSlot _array_element))
    (Utilities|Map|Add (Variables|Default|GetAttachedModsMeshComponents) (GameplayTags|GetTagName (Variables|/Script/FortniteGame.FortWeaponModItemDefinition|GetModSlot (Utilities|Struct|BreakFortWeaponModSlot _array_element))) _returnvalue_2))
  (Rendering|Components|SetLightingChannels (Variables|/Script/FortniteGame.FortWeapon|GetWeaponMesh (Variables|Default|GetPreviewWeapon)) false false true)
  (Helpers|SceneCaptureShowOnlyActors)
  (CallFunction|RepositionCamera))

(fn DestroyScene
  (bind _previewweapon (Variables|Default|GetPreviewWeapon))
  (Utilities|IsValid _previewweapon
    (:Is%20Valid
      (Actor|DestroyActor _previewweapon))
    (:Is%20Not%20Valid
      _))
  (Actor|DestroyActor))

(fn PreviewMod
  (params (Mod%20To%20Preview :type Fort%20Weapon%20Mod%20Item%20Definition%20Object%20Reference) (Selected%20Mod :type Enum_WeaponUICaptureSelectedMod%20Enum :default "NewEnumerator0"))
  (bind _modslot (Variables|/Script/FortniteGame.FortWeaponModItemDefinition|GetModSlot Mod%20To%20Preview))
  (bind _returnvalue (GameplayTags|GetTagName _modslot))
  (Utilities|IsValid (Variables|Default|GetPreviewWeapon)
    (:Is%20Valid
      (bind _returnvalue_2 (Class|FortWeaponModFunctionLibrary|TryAddPreviewWeaponMod Mod%20To%20Preview (Variables|Default|GetPreviewWeapon)))
      (Utilities|Map|Add (Variables|Default|GetAttachedModsMeshComponents) _returnvalue _returnvalue_2)
      (Development|PrintString (Utilities|String|ToString(Name) _returnvalue))
      (Rendering|Components|SetLightingChannels _returnvalue_2 false false true)
      (bind (_mainpass _maskspass) (CaptureWeaponImage|CaptureWeapon false Selected%20Mod))
      (return _mainpass _maskspass))))

(fn RepositionCamera
  (bind (_origin _boxextent) (Collision|GetComponentBounds (Variables|/Script/FortniteGame.FortWeapon|GetWeaponMesh (Variables|Default|GetPreviewWeapon))))
  (bind _previewweapon (Variables|Default|GetPreviewWeapon))
  (Utilities|IsValid _previewweapon
    (:Is%20Valid
      (Transformation|SetWorldLocation (Variables|Default|GetSceneCapture_MainPass) (+ _origin (* (* (Math|Float|Lerp (Variables|Default|GetCamera%20Distance%20Min) (Variables|Default|GetCamera%20Distance%20Max) (Math|Float|NormalizetoRange (* (.y _boxextent) 2.0) 135.0 29.0)) -1.0) (Transformation|GetForwardVector (Variables|Default|GetSceneCapture_MainPass)))) true true))))

(fn ResetMaterials
  (bind _previewweapon (Variables|Default|GetPreviewWeapon))
  (bind _weaponmesh (Variables|/Script/FortniteGame.FortWeapon|GetWeaponMesh _previewweapon))
  (bind _previewweaponoriginalmaterials (Variables|Default|GetPreviewWeaponOriginalMaterials))
  (Utilities|IsValid (Variables|Default|GetPreviewWeapon)
    (:Is%20Valid
      (if (Utilities|Array|IsNotEmpty _previewweaponoriginalmaterials)
        (for (_array_element _array_index) _previewweaponoriginalmaterials
          (Utilities|IsValid _weaponmesh
            (:Is%20Valid
              (Rendering|Material|SetMaterial _weaponmesh _array_index _array_element)))))
      (Utilities|Array|Clear (Variables|Default|GetPreviewWeaponOriginalMaterials))
      (bind _values (Utilities|Map|Values (Variables|Default|GetAttachedModsMeshComponents)))
      (for _array_element_2 _values
        (if (!= (Variables|/Script/Engine.StaticMeshComponent|GetStaticMesh _array_element_2) 0)
          (for (_array_element_3 _array_index_2) (Variables|/Script/Engine.StaticMesh|GetStaticMaterials (Variables|/Script/Engine.StaticMeshComponent|GetStaticMesh _array_element_2))
            (Rendering|Material|SetMaterial _array_element_2 _array_index_2 (Utilities|Struct|BreakStaticMaterial _array_element_3))))))))

(fn ResizeRenderTarget
  (params (Width :type Integer :default 1024) (Height :type Integer :default 512))
  (Variables|Default|SetWidth Width)
  (Variables|Default|SetHeight Height)
  (Rendering|ResizeRenderTarget2D (Variables|Default|GetMainPass) (Variables|Default|GetWidth) (Variables|Default|GetHeight))
  (Rendering|ResizeRenderTarget2D (Variables|Default|GetMasksPass) (Variables|Default|GetWidth) (Variables|Default|GetHeight)))

(fn RevertMod
  (params (FullRevert :type Boolean) (Mod%20To%20Preview :type Fort%20Weapon%20Mod%20Item%20Definition%20Object%20Reference))
  (bind _modslot (Variables|/Script/FortniteGame.FortWeaponModItemDefinition|GetModSlot Mod%20To%20Preview))
  (bind _returnvalue (GameplayTags|GetTagName _modslot))
  (bind _attachedmodsitemdefinition (Variables|Default|GetAttachedModsItemDefinition))
  (bind (_value _returnvalue_2) (Utilities|Map|Find _attachedmodsitemdefinition _returnvalue))
  (bind _previewweapon (Variables|Default|GetPreviewWeapon))
  (bind _attachedmodsitemdefinition_2 (Variables|Default|GetAttachedModsItemDefinition))
  (Utilities|IsValid (Variables|Default|GetPreviewWeapon)
    (:Is%20Valid
      (if FullRevert
        (for _array_element (Class|FortWeapon|GetWeaponModSlots (Variables|Default|GetPreviewWeapon))
          (if (.ReturnValue (Utilities|Map|Find _attachedmodsitemdefinition_2 (GameplayTags|GetTagName (Variables|/Script/FortniteGame.FortWeaponModItemDefinition|GetModSlot (Utilities|Struct|BreakFortWeaponModSlot _array_element)))))
            (Class|FortWeaponModFunctionLibrary|TryAddPreviewWeaponMod (Utilities|Map|Find _attachedmodsitemdefinition_2 (GameplayTags|GetTagName (Variables|/Script/FortniteGame.FortWeaponModItemDefinition|GetModSlot (Utilities|Struct|BreakFortWeaponModSlot _array_element)))) _previewweapon)
            (else
              (Class|FortWeaponModFunctionLibrary|TryRemovePreviewWeaponMod (Utilities|Struct|BreakFortWeaponModSlot _array_element) _previewweapon))))
        (elif _returnvalue_2
          (bind _returnvalue_3 (Class|FortWeaponModFunctionLibrary|TryAddPreviewWeaponMod _value (Variables|Default|GetPreviewWeapon)))
          (Utilities|Map|Add (Variables|Default|GetAttachedModsMeshComponents) _returnvalue _returnvalue_3)
          (else
            (Class|FortWeaponModFunctionLibrary|TryRemovePreviewWeaponMod Mod%20To%20Preview (Variables|Default|GetPreviewWeapon)))))
      (bind (_mainpass _maskspass) (CaptureWeaponImage|CaptureWeapon false))
      (return _mainpass _maskspass))))

(fn SceneCapture
  (params (bMask :type Boolean))
  (Rendering|SceneCapture|CaptureScene (select bMask (Variables|Default|GetSceneCapture_MaskPass) (Utilities|SceneCaptureSelector)) :!class /Script/Engine.SceneCaptureComponent2D)
  (return))

(fn SceneCaptureShowOnlyActors
  (bind _output_get (Variables|/Script/Engine.SceneCaptureComponent|SetShowOnlyActors (Utilities|Array|MakeArray self (Variables|Default|GetPreviewWeapon)) (Utilities|SceneCaptureSelector)))
  (Variables|/Script/Engine.SceneCaptureComponent|SetShowOnlyActors _output_get (Variables|Default|GetSceneCapture_MaskPass)))

(fn SetCaptureVisibility
  (params (bVisible :type Boolean))
  (if (Utilities|ES31Platform)
    (Rendering|SetVisibility (Variables|Default|Getkey_l1_LOW) bVisible false)
    (Rendering|SetActorHiddenInGame (Variables|Default|GetPreviewWeapon) (not bVisible))
    (else
      (Rendering|SetVisibility (Variables|Default|Getfill_l1) bVisible false)
      (Rendering|SetActorHiddenInGame (Variables|Default|GetPreviewWeapon) (not bVisible)))))

(fn ConstructionScript
  (bind _returnvalue (Rendering|CreateRenderTarget2D (/ (Variables|Default|GetWidth) (Math|Integer|SelectInt 2 1 (Utilities|IsMobileorSwitch))) (/ (Variables|Default|GetHeight) (Math|Integer|SelectInt 2 1 (Utilities|IsMobileorSwitch))) "RTF_RGBA8"))
  (bind _output_get (Variables|Default|SetMainPass _returnvalue))
  (Variables|/Script/Engine.SceneCaptureComponent2D|SetTextureTarget _output_get (Utilities|SceneCaptureSelector))
  (bind _returnvalue_2 (Rendering|CreateRenderTarget2D (/ (Variables|Default|GetWidth) (Math|Integer|SelectInt 2 1 (Utilities|IsMobileorSwitch))) (/ (Variables|Default|GetHeight) (Math|Integer|SelectInt 2 1 (Utilities|IsMobileorSwitch))) "RTF_RGBA8"))
  (bind _output_get_2 (Variables|Default|SetMasksPass _returnvalue_2))
  (Variables|/Script/Engine.SceneCaptureComponent2D|SetTextureTarget _output_get_2 (Variables|Default|GetSceneCapture_MaskPass))
  (Transformation|SetActorLocation "0.000000,0.000000,10000.000000" true true)
  ;[1 1 1] Delete components that are not utilized by platform
  (if (Utilities|ES31Platform)
    (Components|DestroyComponent (Variables|Default|Getrim_l1))
    (else
      (Components|DestroyComponent (Variables|Default|Getrim_l1_LOW)))))
