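# Unlit parent material for mask rendering in the weapon mod station's UI capture, usable on skeletal meshes.
# It packs the IsSelected parameter, an inverted dot product of negated camera direction and pixel normal (a Fresnel-like edge term), and a constant one into a float3.
# A separate node scales an Index parameter down by ten.
# The Fortnite_HideElementsWithVertexColor function drives world position offset.
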Material /WeaponModStation/WeaponUICapture/Materials/M_Masks_Parent
graph {
 properties {bUsedWithMeshDeformer=true, bUsedWithSkeletalMesh=true, pixelDepthOffsetMode=PDOM_Legacy, shadingModel=MSM_Unlit, usedShadingModels=Unlit}
 nodes [
  CameraVectorWS()
  Divide(A=ScalarParameter() {parameterName=Index}) {constB=10}
  n1 = Multiply(A=MaterialFunctionCall() {materialFunction=&CameraDirectionVector}.Result) {constB=-1}
  n2 = PixelNormalWS()
  MaterialFunctionCall(X=ScalarParameter() {parameterName=IsSelected, sliderMax=1}, Y=OneMinus(None=DotProduct(A=@n1, B=@n2)), Z=Constant() {r=1}) {materialFunction=&MakeFloat3}
  MaterialFunctionCall!(WPO) {materialFunction=/Game/Packages/Fortress_SharedMaterials/Base_Material_Functions/Fortnite_HideElementsWithVertexColor}
 ]
}
